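# Masked unlit material for the weapons manager Creative device's 2D icon, turned to face the camera through AlignMeshToTheCamera.
# A tiled, clamp-sampled DeviceIcon, offset by IconLocation, adds its squared red channel to the squared ItemIcon color for emissive, with combined alphas as mask.
# The Color tint term subtracts blue from itself, so it contributes zero.
# The Is Hologram switch bypasses MF_CreativeMoveTool when true; otherwise attributes pass through it.
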Material /Game/Creative/Devices/WeaponsManager/Materials/M_CP_Device_2DIcon_Weapon
graph {
 properties {bUseMaterialAttributes=true, blendMode=BLEND_Masked, pixelDepthOffsetMode=PDOM_Legacy, shadingModel=MSM_Unlit, usedShadingModels=Unlit}
 nodes [
  n1 = VectorParameter() {parameterName=IconLocation, defaultValue=<rgba -0.5 0 0 0>}
  n2 = TextureSampleParameter2D!(UVs=Add(A=TextureCoordinate() {uTiling=5, vTiling=5}, B=ComponentMask(None=@n1.RGB) {g=true, r=true}), Level, Apply View MipBias) {
   group=Color
   mipValueMode=TMVM_MipLevel
   parameterName=DeviceIcon
   samplerSource=SSM_Clamp_WorldGroupSettings
   sortPriority=2
   texture=/Game/Creative/Devices/WeaponsManager/Textures/T_CP_WeaponsManager2D_01_D
  }
  n3 = Multiply(A=Multiply(A=Subtract(A=@n2.B, B=@n2.B), B=@n2.A), B=VectorParameter() {parameterName=Color, defaultValue=<rgba 1 0.6967 0 1>}.RGB)
  n4 = Multiply(A=@n2.R, B=@n2.R)
  n5 = TextureSampleParameter2D!(UVs, Apply View MipBias) {
   group=Color
   parameterName=ItemIcon
   samplerSource=SSM_Clamp_WorldGroupSettings
   sortPriority=2
   texture=/Game/UI/Foundation/Textures/Icons/ItemTypes/T-Icon-Blank
  }
  n6 = MakeMaterialAttributes!(
   BaseColor
   Metallic
   Specular
   Roughness
   Anisotropy
   EmissiveColor=Add(A=Add(A=@n3, B=@n4), B=Multiply(A=@n5.RGB, B=@n5.RGB))
   Opacity
   OpacityMask=Clamp(None=Add(A=@n5.A, B=@n2.A))
   Normal
   Tangent
   WorldPositionOffset=MaterialFunctionCall!(
    WorldPosition
    PivotLocation
    WS Normals
    Custom Object Basis 1=Constant3Vector() {constant=<rgba 1 0 0 0>}
    Custom Object Basis 2=Constant3Vector() {constant=<rgba 0 0 -1 0>}
    Custom Object Basis 3=Constant3Vector() {constant=<rgba 0 1 0 0>}
   ) {materialFunction=/Engine/Functions/Engine_MaterialFunctions02/WorldPositionOffset/AlignMeshToTheCamera}.World Position Offset
   SubsurfaceColor
   ClearCoat
   ClearCoatRoughness
   AmbientOcclusion
   Refraction
   CustomizedUVs_0
   CustomizedUVs_1
   CustomizedUVs_2
   CustomizedUVs_3
   CustomizedUVs_4
   CustomizedUVs_5
   CustomizedUVs_6
   CustomizedUVs_7
   PixelDepthOffset
   ShadingModel
   Displacement
  )
  StaticSwitchParameter(True=@n6, False=MaterialFunctionCall(MA In=@n6) {materialFunction=&MF_CreativeMoveTool}.MA Out) {group=Hologram, parameterName=Is Hologram, sortPriority=1}
 ]
}
